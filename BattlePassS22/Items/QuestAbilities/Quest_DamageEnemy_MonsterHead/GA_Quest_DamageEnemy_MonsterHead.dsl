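; Gameplay ability for a Season 22 battle pass quest, apparently about damaging enemies involved with a Monster Head item.
; On an event activation it calls the parent, then checks that the event instigator's pawn is the owning player and tests whether the target is an enemy player pawn.
; The equipped-item tag check is removed to avoid cook warnings, so the team test leads nowhere and no quest credit is granted here.
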(event Ability|EventActivateAbilityFromEvent
  (params (EventData :type Gameplay%20Event%20Data%20Structure%20%28by%20ref%29))
  (bind (_eventtag _instigator _target) (Utilities|Struct|BreakGameplayEventData EventData))
  ;[0 0 0] Ensure the owning player is the instigator of the damage
  (|CallParentFunction EventData)
  (bind _asfort_player_state_athena (Utilities|Casting|CastToFortPlayerStateAthena _instigator)
    (:then
      (if (== (PlayerState|GetCurrentPawn _asfort_player_state_athena) (Variables|GAVars|Data|GetOwningPlayerPawn))
        ;[0 0 0] Ensure the instigator hit an enemy player
        (bind _asfort_player_pawn (Utilities|Casting|CastToFortPlayerPawn _target)
          (:then
            (if (not (Team|OnSameTeam _asfort_player_pawn (Variables|GAVars|Data|GetOwningPlayerPawn)))
              _
              (else
                _)))
          (:CastFailed
            _))
        ;[0 0 0] Check if enemy player has the item equipped
        ;;[1 0.2656 0.2656] Removed for Cook Warnings! Logic is to get owned gameplay tags and check for the granted tag from the item
        (Ability|EndAbility)
        (else
          _)))
    (:CastFailed
      _)))
